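# Animation Blueprint for the CanyonVale backpack cosmetic that procedurally adjusts its pack bone to follow the wearer's torso bend and vertical motion.
# Each update reads the owner mesh's spine_05-to-spine_01 delta from ref pose and remaps its rotation into rotation, translation and scale variables.
# The AnimGraph chains additive Modify Bone nodes applying these plus a fixed offset, with a roll weighted by pawn Z velocity.
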AnimBlueprint /BRCosmetics/Accessories/FORT_Backpacks/Backpack_M_MED_CanyonVale/Meshes/M_MED_CanyonVale_Pack_AnimBP
target_skeleton=&/M_MED_CanyonVale_Pack_Skeleton
graphs [
 AnimGraph {nodes [
   n1 = "Transform (Modify) Bone - Bone: pack"(
    ComponentPose=Local To Component(LocalPose=Input Pose())
    Alpha=Get trampoline() -> (trampoline:double)
    Translation="0,0,0"
    Rotation="0,0,5"
   ) {
    rotationMode=BMM_Additive
    translationMode=BMM_Additive
    alphaBoolBlend {blendOption=Linear}
    boneToModify {boneName=pack}
    rotation {roll=5}
   }
   n2 = "Transform (Modify) Bone - Bone: pack"(ComponentPose=@n1, Translation="0,0.5,0", Rotation="0,0,0") {
    rotationMode=BMM_Additive
    translationMode=BMM_Additive
    alphaBoolBlend {blendOption=Linear}
    boneToModify {boneName=pack}
    translation {y=0.5}
   }
   n3 = "Transform (Modify) Bone - Bone: pack"(ComponentPose=@n2, Translation="0,0,0", Rotation="0,0,0", Scale=Get scale() -> (scale:Vector)) {
    rotationMode=BMM_Additive
    scaleMode=BMM_Additive
    translationMode=BMM_Additive
    alphaBoolBlend {blendOption=Linear}
    boneToModify {boneName=pack}
   }
   n4 = "Transform (Modify) Bone - Bone: pack"(
    ComponentPose=@n3
    Translation=Get backpack_translation_Z() -> (backpack_translation_Z:Vector)
    Rotation=Get pickaxe_rotate() -> (pickaxe_rotate:Rotator)
   ) {rotationMode=BMM_Additive, translationMode=BMM_Additive, alphaBoolBlend {blendOption=Linear}, boneToModify {boneName=pack}}
   n5 = "Transform (Modify) Bone - Bone: pack"(ComponentPose=@n4, Translation=Get backpack_translation_Y() -> (backpack_translation_Y:Vector), Rotation=Get backpack_rotate()) {rotationMode=BMM_Additive, translationMode=BMM_Additive, alphaBoolBlend {blendOption=Linear}, boneToModify {boneName=pack}}
   Output Pose(Result=Component To Local(ComponentPose=@n5)) {layerGroup=None}
  ]}
 EventGraph {nodes [
   Event BlueprintUpdateAnimation()
   n1 = GetOwnerSkeletalMeshComponent()
   n2 = BreakRotator(InRot=BreakTransform(InTransform=GetDeltaTransformFromRefPose(self=@n1, BoneName=spine_05, BaseName=spine_01)).Rotation)
   > Set backpack_rotate(backpack_rotate=MakeRotator(Roll=MapRangeClamped(Value=@n2.Yaw, InRangeA=-10, InRangeB=-60, OutRangeB=50)))
   > Set backpack_translation_Y(backpack_translation_Y:Vector=MakeVector(Y=MapRangeClamped(Value=@n2.Yaw, InRangeA=-10, InRangeB=-60, OutRangeB=-15))) -> (Output_Get:Vector)
   > Set backpack_translation_Z(backpack_translation_Z:Vector=MakeVector(Y=0, Z=MapRangeClamped(Value=@n2.Yaw, InRangeA=0, InRangeB=-60, OutRangeB=0))) -> (Output_Get:Vector)
   > Set pickaxe_rotate(pickaxe_rotate:Rotator=MakeRotator(Yaw=MapRangeClamped(Value=@n2, InRangeA=0, InRangeB=46, OutRangeB=-18))) -> (Output_Get:Rotator)
   n3 = MapRangeClamped(Value=@n2.Yaw, InRangeA=0, InRangeB=-60, OutRangeA=1, OutRangeB=0.9)
   > Set scale(scale:Vector=MakeVector(X=@n3, Y=@n3, Z=@n3)) -> (Output_Get:Vector)
   > Set trampoline(trampoline:double=MapRangeClamped(Value=Get PawnVelocityZ(), InRangeA=0, InRangeB=4000)) -> (Output_Get:double)
   TryGetPawnOwner()
  ]}
]
variables [
 backpack_rotate {type=Rotator}
 backpack_translation_Y {type=Vector}
 backpack_translation_Z {type=Vector}
 isGelatinGummiEquipped {type=boolean}
 pickaxe_rotate {type=Rotator}
 scale {type=Vector}
 trampoline {type=number}
]
